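# Sound cue for the DonutCup pickaxe cosmetic's impact on enemies.
# A short fixed delay feeds a modulator that slightly varies pitch and volume, which wraps an equal-weight random pick among four impact wave variations.
# It uses the shared pickaxe impact attenuation, property sheet and melee weapon concurrency set, plays in the impacts sound class, and is primed on load.
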SoundCue /BRCosmetics/Sounds/PickAxes/DonutCup/Pickaxe_DonutCup_EnemyImpact_Cue
root=Delay_1
nodes [
 {id=Delay_1, children [Modulator_1], properties {delayMax=0.07, delayMin=0.07}}
 {id=Modulator_1, children [Random_1], properties {pitchMax=1.1, pitchMin=1, volumeMax=0.9, volumeMin=0.8}}
 {id=Random_1, children [WavePlayer_5,WavePlayer_6,WavePlayer_7,WavePlayer_9], properties {weights [1,1,1,1]}}
 {id=WavePlayer_5, properties {soundWaveAssetPtr=&/Pickaxe_DonutCup_EnemyImpact_01}}
 {id=WavePlayer_6, properties {soundWaveAssetPtr=&/Pickaxe_DonutCup_EnemyImpact_02}}
 {id=WavePlayer_7, properties {soundWaveAssetPtr=&/Pickaxe_DonutCup_EnemyImpact_03}}
 {id=WavePlayer_9, properties {soundWaveAssetPtr=&/Pickaxe_DonutCup_EnemyImpact_05}}
]
properties {
 attenuationSettings=/Game/Sounds/Attenuation/Fort_Pickaxe_Impact_Att
 audioPropertiesSheet=/BRCosmetics/Sounds/Settings/PropertySheets/Pickaxe/APS_Cosmetic_Pickaxe_Impact
 bPrimeOnLoad=true
 maxDistance=4300
 pitchMultiplier=1.05
 soundClassObject=&Fort_Impacts
 attenuationOverrides {
  bAttenuateWithLPF=true
  distanceAlgorithm=Logarithmic
  falloffDistance=800
  lPFRadiusMax=800
  lPFRadiusMin=400
  attenuationShapeExtents {x=300}
 }
 concurrencyOverrides {bLimitToOwner=true, maxCount=2, resolutionRule=StopOldest}
 concurrencySet [&Player_Melee_Weapon_Conc]
}
